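# Animation Blueprint for the Knight Cat BuiltIn Female pickaxe skeleton that keeps the axe's eight blade bones collapsed by default.
# Starting from the ref pose, chained Transform Bone nodes replace the scale of Melee_TwoHanded_Blade1 through Blade8 with zero.
# The result feeds DefaultSlot ahead of the output, so animations played in that slot can override the collapsed pose.
# The event graph holds only unconnected default nodes.
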AnimBlueprint /BRCosmetics/Weapons/FORT_Melee/Pickaxe_Knight_Cat_BuiltIn_Female/Meshes/Knight_Cat_BuiltIn_Female_Axe_Skeleton_AnimBlueprint
target_skeleton=&/Knight_Cat_BuiltIn_Female_Axe_Skeleton
graphs [
 AnimGraph {nodes [
   n1 = "Transform (Modify) Bone - Bone: Melee_TwoHanded_Blade1"(ComponentPose=Local To Component(LocalPose=Local Space Ref Pose()), Alpha=1, Translation="0,0,0", Rotation="0,0,0", Scale="0,0,0") {scaleMode=BMM_Replace, alphaBoolBlend {blendOption=Linear}, boneToModify {boneName=Melee_TwoHanded_Blade1}, scale=<0 0 0>}
   n2 = "Transform (Modify) Bone - Bone: Melee_TwoHanded_Blade2"(ComponentPose=@n1, Alpha=1, Translation="0,0,0", Rotation="0,0,0", Scale="0,0,0") {scaleMode=BMM_Replace, alphaBoolBlend {blendOption=Linear}, boneToModify {boneName=Melee_TwoHanded_Blade2}, scale=<0 0 0>}
   n3 = "Transform (Modify) Bone - Bone: Melee_TwoHanded_Blade3"(ComponentPose=@n2, Alpha=1, Translation="0,0,0", Rotation="0,0,0", Scale="0,0,0") {scaleMode=BMM_Replace, alphaBoolBlend {blendOption=Linear}, boneToModify {boneName=Melee_TwoHanded_Blade3}, scale=<0 0 0>}
   n4 = "Transform (Modify) Bone - Bone: Melee_TwoHanded_Blade4"(ComponentPose=@n3, Alpha=1, Translation="0,0,0", Rotation="0,0,0", Scale="0,0,0") {scaleMode=BMM_Replace, alphaBoolBlend {blendOption=Linear}, boneToModify {boneName=Melee_TwoHanded_Blade4}, scale=<0 0 0>}
   n5 = "Transform (Modify) Bone - Bone: Melee_TwoHanded_Blade5"(ComponentPose=@n4, Alpha=1, Translation="0,0,0", Rotation="0,0,0", Scale="0,0,0") {scaleMode=BMM_Replace, alphaBoolBlend {blendOption=Linear}, boneToModify {boneName=Melee_TwoHanded_Blade5}, scale=<0 0 0>}
   n6 = "Transform (Modify) Bone - Bone: Melee_TwoHanded_Blade6"(ComponentPose=@n5, Alpha=1, Translation="0,0,0", Rotation="0,0,0", Scale="0,0,0") {scaleMode=BMM_Replace, alphaBoolBlend {blendOption=Linear}, boneToModify {boneName=Melee_TwoHanded_Blade6}, scale=<0 0 0>}
   n7 = "Transform (Modify) Bone - Bone: Melee_TwoHanded_Blade7"(ComponentPose=@n6, Alpha=1, Translation="0,0,0", Rotation="0,0,0", Scale="0,0,0") {scaleMode=BMM_Replace, alphaBoolBlend {blendOption=Linear}, boneToModify {boneName=Melee_TwoHanded_Blade7}, scale=<0 0 0>}
   n8 = "Transform (Modify) Bone - Bone: Melee_TwoHanded_Blade8"(ComponentPose=@n7, Alpha=1, Translation="0,0,0", Rotation="0,0,0", Scale="0,0,0") {scaleMode=BMM_Replace, alphaBoolBlend {blendOption=Linear}, boneToModify {boneName=Melee_TwoHanded_Blade8}, scale=<0 0 0>}
   Output Pose(Result=Slot 'DefaultSlot'(Source=Component To Local(ComponentPose=@n8))) {layerGroup=None}
  ]}
 EventGraph {nodes [
   TryGetPawnOwner()
   Event BlueprintUpdateAnimation()
  ]}
]
variables []
